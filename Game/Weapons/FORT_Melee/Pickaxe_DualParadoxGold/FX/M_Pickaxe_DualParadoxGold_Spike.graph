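# Translucent unlit Niagara sprite material for the spike effect of the DualParadoxGold pickaxe.
# A rotatable spike texture, shaped by a power and tinted by particle color, feeds emissive adjustments.
# UseKneeLens blends in a multiply by a camera-vector triplanar layer of dim grayscale nebula with gold stars, off by default.
# Opacity combines spike and particle alpha with a panning noise dissolve, offset per particle and driven by a dynamic parameter.
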Material /Game/Weapons/FORT_Melee/Pickaxe_DualParadoxGold/FX/M_Pickaxe_DualParadoxGold_Spike
graph {
 properties {bUsedWithNiagaraSprites=true, blendMode=BLEND_Translucent, pixelDepthOffsetMode=PDOM_Legacy, shadingModel=MSM_Unlit, usedShadingModels=Unlit}
 nodes [
  n1 = TextureSample(UVs=MaterialFunctionCall(
    UVs=TextureCoordinate()
    Rotation Center=Constant() {r=0.5}
    "Rotation Angle (0-1)"=ScalarParameter() {parameterName="Rotation Angle (0-1) (S)"}
   ) {materialFunction=&CustomRotator}.Rotated Values) {texture=&/T_Pickaxe_DualParadoxGold_Spike_01}
  n2 = ParticleColor()
  n3 = Multiply(A=Power(Base=@n1.RGB, Exp=ScalarParameter() {defaultValue=1, parameterName=Power}) {constExponent=1}, B=@n2.RGB)
  n4 = NamedRerouteUsage()
  MaterialFunctionCall(Emissive=LinearInterpolate(A=@n3, B=Multiply(A=@n3, B=@n4), Alpha=ScalarParameter() {group=KneeLensParms, parameterName=UseKneeLens})) {materialFunction=&MF_FX_EmissiveAdjustments}
  n5 = ParticleRandom()
  n6 = VectorParameter() {group=Dissolve, parameterName=Dissolve Tex UV, sortPriority=6, defaultValue=<rgba 0.1 0.1 0 0>}
  n7 = AppendVector(A=@n6.R, B=@n6.R)
  n8 = TextureCoordinate() {uTiling=4}
  n9 = Add(A=AppendVector(A=@n5, B=@n5), B=Panner(Coordinate=Multiply(A=@n8, B=@n7), Speed=AppendVector(A=@n6.B, B=@n6.B)))
  n10 = ScalarParameter() {defaultValue=3, group=Dissolve, parameterName=Dissolve Edge Hardness, sliderMin=0.5}
  n11 = Subtract(
   A=Multiply(
    A=TextureSampleParameter2D(UVs=@n9) {group=Dissolve, parameterName=Dissolve Tex, texture=/Game/Global/Textures/Default/Noise/T_FireNoise_01}.R
    B=@n10
   )
   B=LinearInterpolate(A=@n10, Alpha=DynamicParameter() {defaultValue=<rgba 0 1 1 1>, paramNames [Random,Dissolve,Param3,Param4]}.Dissolve) {constB=-1}
  )
  Multiply(A=Multiply(A=@n1.A, B=Saturate(None=@n11)), B=@n2.A)
  n12 = TextureObjectParameter() {
   group=KneeLensParms
   parameterName=GalaxyTexture
   texture=/Game/Characters/Player/Male/Medium/Bodies/M_Med_Soldier_04/Skins/Galaxy/Textures/T_FN_Nebula
  }
  n13 = Time()
  n14 = Multiply(A=@n13, B=Constant())
  n15 = AppendVector(
   A=ScalarParameter() {defaultValue=0.02, group=KneeLensParms, parameterName=Galaxy Rotation Speed}
   B=ScalarParameter() {defaultValue=-0.02, group=KneeLensParms, parameterName=Galaxy Rotation Speed_1}
  )
  n16 = ScalarParameter() {defaultValue=0.1, group=KneeLensParms, parameterName=Galaxy Rotation Speed_2}
  n17 = MaterialFunctionCall!(
   CameraVector
   Rotation Speed=Multiply(A=LinearInterpolate(A=@n13, B=@n14) {constAlpha=0}, B=AppendVector(A=@n15, B=@n16))
   Texture=@n12
   Tiling=ScalarParameter() {defaultValue=1, group=KneeLensParms, parameterName=GalaxyTiling}
   Offset=Constant()
   AxisFadeContrast=ScalarParameter() {defaultValue=6, group=KneeLensParms, parameterName=GalaxyAxisFade}
  ) {materialFunction=/Game/Weapons/WeaponSkins/MaterialFunction/TriplanarCameraVector_Rotated}
  n18 = ScalarParameter() {defaultValue=-20, group=KneeLensParms, parameterName=Galaxy Saturation}
  n19 = Desaturation(
   None=Multiply(
    A=VectorParameter() {group=KneeLensParms, parameterName=Galaxy Color, defaultValue=<rgba 1 1 1 1>}.RGB
    B=Desaturation(None=@n17.Result, Fraction=@n18)
   )
   Fraction=Constant() {r=1}
  )
  n20 = Constant3Vector() {constant=<rgba 1 0.7816 0.0452 1>}
  n21 = ScalarParameter() {defaultValue=1, group=KneeLensParms, parameterName=KneeLense_StarBrightness}
  n22 = Time()
  n23 = ScalarParameter() {group=KneeLensParms, parameterName=Small Star Rotation speed}
  n24 = TextureObject() {samplerType=SAMPLERTYPE_Masks, texture=/Game/Characters/Player/Male/Medium/Bodies/M_Med_Soldier_04/Skins/Galaxy/Textures/Celestial_T_Stars}
  n25 = ComponentMask(None=MaterialFunctionCall(
    Rotation Speed=Multiply(A=@n22, B=@n23)
    Texture=@n24
    Tiling=ScalarParameter() {defaultValue=8, group=KneeLensParms, parameterName=Small Star Tiling}
   ) {materialFunction=&TriplanarCameraVector_Rotated}.Result) {b=true}
  n26 = Multiply(A=@n25, B=@n25)
  n27 = TextureObject() {samplerType=SAMPLERTYPE_Masks, texture=&Celestial_T_Stars}
  n28 = ComponentMask(None=MaterialFunctionCall!(
    CameraVector
    Texture=@n27
    Tiling=ScalarParameter() {defaultValue=2, group=KneeLensParms, parameterName=Bright Star Tiling}
    Offset
    AxisFadeContrast
   ) {materialFunction=/Engine/Functions/Engine_MaterialFunctions01/Texturing/TriplanarCameraVector}.Result) {r=true}
  n29 = TextureSample(UVs=Panner(Coordinate=TextureCoordinate() {uTiling=2, vTiling=2}) {speedX=0.133, speedY=0.5}) {
   samplerSource=SSM_Wrap_WorldGroupSettings
   samplerType=SAMPLERTYPE_Masks
   texture=&Celestial_T_Stars
  }
  n30 = Multiply(A=Multiply(A=@n28, B=@n28), B=@n29.G)
  NamedRerouteDeclaration(None=LinearInterpolate(A=Multiply(A=@n19) {constB=0.1}, B=Multiply(A=@n20, B=@n21), Alpha=Add(A=@n26, B=@n30))) {name=KneeLens, nodeColor=<rgba 0.0312 0 1 1>}
 ]
}
